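# Unlit, two-sided translucent master material for environment fog in the EngineVolumetrics plugin, with depth testing disabled.
# Fog density comes from scene depth with start distance and falloff, multiplied by a height gradient from UV2 and a blended opacity.
# Two color and opacity parameter sets are interpolated, alongside a blend gradient built from UV1 with offset and power.
# A static switch enables side desaturation of the blended color.
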Material /Engine/EngineVolumetrics/FogEnvironment/Materials/M_EV_FogEnvironment_Master_01
graph {
 properties {
  bDisableDepthTest=true
  bUsedWithStaticLighting=true
  blendMode=BLEND_Translucent
  pixelDepthOffsetMode=PDOM_Legacy
  shadingModel=MSM_Unlit
  twoSided=true
  usedShadingModels=Unlit
 }
 nodes [
  n1 = ScalarParameter() {defaultValue=0.5, parameterName=BlendOffset}
  n2 = Subtract(A=ComponentMask(None=TextureCoordinate() {coordinateIndex=1}) {r=true}, B=@n1)
  n3 = ScalarParameter() {defaultValue=4, parameterName=BlendPower}
  n4 = Constant() {r=0.5}
  n5 = Subtract(A=Subtract(A=Multiply(A=@n2, B=@n3), B=Multiply(A=@n4, B=@n1)), B=@n4)
  Add(A=OneMinus(None=Abs(None=@n5)), B=Constant() {r=0.25})
  n6 = SceneDepth()
  Clamp(None=@n6.R)
  Multiply(B=ScalarParameter() {defaultValue=0.5, parameterName=SideDesaturation})
  n7 = Add(A=ComponentMask(None=TextureCoordinate() {coordinateIndex=2}) {g=true}, B=ScalarParameter() {parameterName=HeightOffset})
  n8 = ScalarParameter() {defaultValue=2, parameterName=HeightDensity}
  n9 = LinearInterpolate(A=ScalarParameter() {defaultValue=1, desc=Opacity, parameterName=Opacity1}, B=ScalarParameter() {defaultValue=1, parameterName=Opacity2})
  n10 = ScalarParameter() {parameterName=StartDistance}
  n11 = Clamp(None=Multiply(A=Add(A=@n6.R, B=@n10), B=ScalarParameter() {defaultValue=0.0001, parameterName=Density}))
  n12 = ScalarParameter() {defaultValue=1, parameterName=Falloff}
  Multiply(A=Multiply(A=Power(Base=@n7, Exp=@n8), B=Multiply(A=@n8, B=@n8)), B=Multiply(A=Clamp(None=@n9), B=Power(Base=@n11, Exp=@n12)))
  Power(Exp=ScalarParameter() {defaultValue=1, parameterName=DesatPower})
  n13 = VectorParameter() {desc="Color (A is luminosity)", parameterName=Color1, defaultValue=<rgba 1 1 1 1>}
  n14 = VectorParameter() {desc="Color (A is luminosity)", parameterName=Color2, defaultValue=<rgba 1 1 1 1>}
  n15 = LinearInterpolate(A=Multiply(A=@n13.RGB, B=@n13.RGB), B=Multiply(A=@n14.RGB, B=@n14.RGB))
  StaticSwitchParameter(True=Desaturation(None=@n15), False=@n15) {parameterName=UseSideDesaturation}
 ]
}
